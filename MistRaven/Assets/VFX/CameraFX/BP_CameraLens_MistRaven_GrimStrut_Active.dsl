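; Camera lens effect Blueprint for MistRaven's active GrimStrut ability, driving a Niagara effect.
; On BeginPlay it feeds the GrimStrut duration, read from a scalable float, into the Niagara component's LoopTime parameter.
; A custom EndingGrimStrutCameraPP event sets the Niagara End flag to true to signal the effect's finish.
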(event EventBeginPlay
  (Niagara|SetNiagaraVariable(Float) (Variables|NiagaraActor|GetNiagaraComponent) "LoopTime" (ScalableFloat|GetValueatLevel (Variables|Default|GetGrimStrutDuration))))

(event Custom|EVENT_EndingGrimStrutCameraPP
  (Niagara|SetNiagaraVariable(Bool) (Variables|NiagaraActor|GetNiagaraComponent) "End" true))

(fn ConstructionScript)
