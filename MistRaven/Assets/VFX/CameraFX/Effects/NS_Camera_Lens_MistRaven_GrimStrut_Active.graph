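# Camera lens Niagara effect for the active state of MistRaven's GrimStrut, built from speedline sprites and mesh-based tendril and wipe overlays.
# Speedlines spawn continuously until User.End is set, with opacity scaled by time of day.
# Tendrils and Wipe are single particles lasting LoopTime plus TimeAdjustment, with age curves driving material parameters and Wipe scale.
# User.End shifts those curve indices using system end time, so the outro can play early; InnerFlash is disabled.
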NiagaraSystem /MistRaven/Assets/VFX/CameraFX/Effects/NS_Camera_Lens_MistRaven_GrimStrut_Active
dependencies {
 usedDataInterfaces [&NiagaraDataInterfaceCurve, &NiagaraDataInterfaceVector2DCurve, &NiagaraDataInterfaceArrayColor]
 usedDynamicInputs []
 usedModules [
  &self:SystemSpawnScript.SystemScriptSource.SystemScriptGraph.NiagaraNodeAssignment_1.SetVariables_#1
  &SystemState
  &self:ScratchModule
 ]
}
emitters {
 InnerFlash {
  bEnabled=false
  emitterUpdateScript [
   &EmitterState(Life Cycle Mode=::NewEnumerator1, Loop Behavior=::NewEnumerator1, Loop Duration=4, UseLoopDelay=false)
   &SpawnBurst_Instantaneous(Spawn Time=0.08, Use Spawn Probability=false, UseLoopCountLimit=false)
  ]
  particleSpawnScript [
   &InitializeParticle(
    Color=<rgba 0.2 0.5 1 1>
    Lifetime=&Multiply_Float(A=0.4, B=$User.LifetimeScale)
    Sprite Size Mode=::NewEnumerator3
    Uniform Sprite Size=150
    UsePositionOffset=false
    Write Material Random=true
   )
   &CameraOffset(Camera Offset Amount=50)
  ]
  particleUpdateScript [
   &ParticleState()
   &ScaleSpriteSize(
    Uniform Curve Scale=&Multiply_Float(A=1, B=$User.ScaleSize)
    Uniform Curve Sprite Scale=DataInterface() {propertyValues {Curve {keys [
        {arriveTangent=9.0725, interpMode=RCIM_Cubic, leaveTangent=9.0725, tangentMode=RCTM_User, time=0, value=0}
        {arriveTangent=-2.5249, interpMode=RCIM_Cubic, leaveTangent=-2.5249, tangentMode=RCTM_User, time=0.1408, value=0.9847}
        {arriveTangent=-0.0405, interpMode=RCIM_Cubic, leaveTangent=-0.0405, tangentMode=RCTM_User, time=1, value=0.45}
       ]}}}
   )
   &ScaleColor(
    Scale Alpha=&FloatFromCurve(
     FloatCurve=DataInterface() {propertyValues {Curve {keys [
         {arriveTangent=-0.5632, interpMode=RCIM_Cubic, leaveTangent=-0.5632, tangentMode=RCTM_User, time=0, value=1}
         {arriveTangent=-1, interpMode=RCIM_Cubic, leaveTangent=-1, time=0.2, value=0.4}
         <at 1 0>
        ]}}}
     Scale Curve=&Multiply_Float(A=0.3, B=$User.ScaleOpacity)
    )
    Scale RGB=&VectorFromFloat(Value=&FloatFromCurve(
      FloatCurve=DataInterface() {propertyValues {Curve {keys [
          {interpMode=RCIM_Cubic, tangentMode=RCTM_User, time=0, value=1}
          {arriveTangent=-1, interpMode=RCIM_Cubic, leaveTangent=-1, time=0.2, value=0.4}
          {interpMode=RCIM_Cubic, tangentMode=RCTM_User, time=1, value=0}
         ]}}}
      Scale Curve=0.3
     ))
    ScaleA=true
    ScaleRGB=true
   )
  ]
  rendererClasses [&NiagaraSpriteRendererProperties]
  renderers [{rendererClass=&NiagaraSpriteRendererProperties}]
 }
 Speedlines {
  emitterUpdateScript [
   &EmitterState()
   &SpawnBurst_Instantaneous(Spawn Count=20, Use Spawn Probability=false, UseLoopCountLimit=false)
   &SpawnRate(
    Spawn Probability=&MakeCustomFloatFromBool(Bool=$User.End, True Float=0, False Float=1)
    SpawnRate=375
    Use Spawn Probability=true
   )
  ]
  particleSpawnScript [
   /Niagara/Modules/Spawn/Initialization/InitializeParticle(
    Color=&LerpLinearColors(
     StartColor=&SelectLinearColorFromArray(
      Array Sampling Mode=::NewEnumerator1
      Direct Array Index=$User.ColorVariant
      Color Selection Array=DataInterface() {propertyValues {ColorData [<rgba 0 0.4025 1.5 1>, <rgba 0.625 0 0.04 1>, <rgba 1 0.3072 0 1>, <rgba 0.3522 0 1 1>]}}
      Random Seed=?
     )
     EndColor=<rgba 0 0 0 1.6>
     LerpFactor=&UniformAOrBFloat(A=0, Distribution Weight=0.4, Evaluation Type=::NewEnumerator0, Recalculate Random Each Loop=true, Override Seed=false)
    )
    Lifetime=&UniformRangedFloat(Minimum=0.1, Maximum=0.15)
    Module.Write Color=true
    Module.Write Lifetime=true
    Module.Write Mass=true
    Module.Write Material Random=true
    Module.Write Position=false
    Module.Write Scale=false
    Module.Write SpriteRotation=false
    Module.Write SpriteSize=true
    Position=&SimulationPosition()
    Sprite Size=&UniformRangedVector2D(Minimum=<4 400>, Maximum=<1 500>)
   ) {enabled=false}
   /Niagara/Modules/Spawn/Initialization/V2/InitializeParticle(
    Color=&LerpLinearColors(
     StartColor=<rgba 0.1823 0.1823 0.1823 1>
     EndColor=<rgba 1 0 0.1234 1.6>
     LerpFactor=&UniformAOrBFloat(A=0, Distribution Weight=0.4, Evaluation Type=::NewEnumerator0, Recalculate Random Each Loop=true, Override Seed=false)
    )
    Lifetime=&UniformRangedFloat(Minimum=0.1, Maximum=0.15)
    Mass Mode=::NewEnumerator1
    Sprite Size=&UniformRangedVector2D(Minimum=<2 400>, Maximum=<1 500>)
    Sprite Size Mode=::NewEnumerator4
    UsePositionOffset=false
    Write Material Random=true
   ) {module=InitializeParticle002}
   /Niagara/Modules/Spawn/Initialization/InitializeParticle(
    Color=&LerpLinearColors(
     StartColor=<rgba 1 1 1 1.6>
     EndColor=<rgba 0 0 0 1.6>
     LerpFactor=&UniformAOrBFloat(A=0, Distribution Weight=0.15, Evaluation Type=::NewEnumerator0, Recalculate Random Each Loop=true, Override Seed=false)
    )
    Lifetime=&UniformRangedFloat(Minimum=0.1, Maximum=0.15)
    Module.Write Color=true
    Module.Write Lifetime=true
    Module.Write Mass=true
    Module.Write Material Random=true
    Module.Write Position=false
    Module.Write Scale=false
    Module.Write SpriteRotation=false
    Module.Write SpriteSize=true
    Position=&SimulationPosition()
    Sprite Size=&UniformRangedVector2D(Minimum=<4 400>, Maximum=<1 500>)
   ) {enabled=false, module=InitializeParticle001}
   &ShapeLocation(
    Handle Radius=0
    Large Radius=&RandomRangeFloat(Minimum=17, Maximum=23)
    MaskSpawnGroup=false
    Non Uniform Scale=<1 1.7 1>
    Offset Mode=::NewEnumerator0
    Override Seed=false
    Rotation Mode=::NewEnumerator1
    Shape Primitive=::NewEnumerator3
    UseOwnerScale=true
   )
   &AddVelocityFromPoint(Module.Use Falloff Distance=false, Velocity Strength=&UniformRangedFloat(Minimum=1000, Maximum=300))
   &CameraOffset(Camera Offset Amount=-100)
  ]
  particleUpdateScript [
   &ParticleState()
   &SolveForcesAndVelocity(Clamp Velocity=false, Limit Acceleration=false)
   &ScaleSpriteSize(Scale Factor=&Vector2DFromCurve(
     Vector2Curve=DataInterface() {propertyValues {XCurve {keys [<at 0 0.5 RCIM_Cubic>,<at 0.5 1 RCIM_Cubic>]}, YCurve {keys [<at 0 1 RCIM_Cubic>,<at 0.5 1 RCIM_Cubic>]}}}
     ScaleCurve=&Vector2DFromFloat(Value=1)
    ))
   &ScaleColor(
    Scale Alpha=&Multiply_Float(
     A=&FloatFromCurve(
      FloatCurve=DataInterface() {propertyValues {Curve {keys [
          {arriveTangent=5.2641, interpMode=RCIM_Cubic, leaveTangent=5.2641, tangentMode=RCTM_User, time=0, value=0}
          {arriveTangent=0.7132, interpMode=RCIM_Cubic, leaveTangent=0.7132, tangentMode=RCTM_User, time=0.3, value=1.0207}
          {arriveTangent=-2.9069, interpMode=RCIM_Cubic, leaveTangent=-2.9069, tangentMode=RCTM_User, time=1, value=0}
         ]}}}
      Scale Curve=5
     )
     B=&FNTimeOfDayScalar(Night Scale=0.5, Morning Scale=0.25, Evening Scale=0.15)
    )
    Scale RGB=&VectorFromFloat(Value=1)
    ScaleA=true
    ScaleRGB=true
   )
  ]
  rendererClasses [&NiagaraSpriteRendererProperties]
  renderers [{rendererClass=&NiagaraSpriteRendererProperties}]
 }
 Tendrils {
  emitterUpdateScript [
   &EmitterState()
   &SpawnBurst_Instantaneous(Spawn Count=1, Use Spawn Probability=false)
  ]
  particleSpawnScript [
   &CameraOffset(Camera Offset Amount=6) {enabled=false}
   /Niagara/Modules/Spawn/Initialization/V2/InitializeParticle(
    Color=<rgba 0.5667 0 2 3>
    Lifetime=&Add_Float(A=$User.LoopTime, B=$User.TimeAdjustment)
    Mass Mode=::NewEnumerator1
    Position Offset=<-250 0 0>
    UsePositionOffset=false
    Write Material Random=true
   ) {module=InitializeParticle001}
  ]
  particleUpdateScript [
   &ParticleState()
   &DynamicMaterialParameters(
    Index 0 Param 1=&FloatFromCurve(
     FloatCurve=DataInterface() {propertyValues {Curve {
        defaultValue=1
        keys [
         {
          arriveTangentWeight=0.0267
          interpMode=RCIM_Cubic
          leaveTangentWeight=0.0267
          tangentMode=RCTM_User
          tangentWeightMode=RCTWM_WeightedBoth
          time=0
          value=0
         }
         {arriveTangentWeight=0.0267, interpMode=RCIM_Cubic, tangentMode=RCTM_User, time=0.08, value=1}
         {arriveTangentWeight=0.28, interpMode=RCIM_Cubic, tangentMode=RCTM_User, time=0.92, value=1}
         {interpMode=RCIM_Cubic, leaveTangent=-12.0916, time=1, value=0}
        ]
       }}}
     CurveIndex=&MakeCustomFloatFromBool(
      Bool=$User.End
      True Float=&Add_Float(
       A=$Particles.NormalizedAge
       B=&Subtract_Float(
        A=&MakeCustomFloatFromBool(
         Bool=&SetBoolByFloatComparison(A=0.92, B=$Particles.NormalizedAge, Comparison Type=::NewEnumerator6)
         True Float=$Particles.NormalizedAge
         False Float=0.92
        )
        B=&Divide_Float(A=$System.endtime, B=$System.looptimeInit)
       )
      )
      False Float=$Particles.NormalizedAge
     )
    )
    Index 0 Param 2=&FloatFromCurve(
     FloatCurve=DataInterface() {propertyValues {Curve {
        defaultValue=1
        keys [
         <at 0 0 RCIM_Cubic>
         {arriveTangentWeight=0.0599, interpMode=RCIM_Cubic, tangentMode=RCTM_User, time=0.18, value=1}
         {arriveTangentWeight=0.1992, interpMode=RCIM_Cubic, tangentMode=RCTM_User, time=0.82, value=1}
         {arriveTangentWeight=0.0667, interpMode=RCIM_Cubic, tangentMode=RCTM_User, time=1, value=0}
        ]
       }}}
     CurveIndex=&MakeCustomFloatFromBool(
      Bool=$User.End
      True Float=&Add_Float(
       A=$Particles.NormalizedAge
       B=&Subtract_Float(
        A=&MakeCustomFloatFromBool(
         Bool=&SetBoolByFloatComparison(A=0.82, B=$Particles.NormalizedAge, Comparison Type=::NewEnumerator6)
         True Float=$Particles.NormalizedAge
         False Float=0.82
        )
        B=&Divide_Float(A=$System.endtime, B=$System.looptimeInit)
       )
      )
      False Float=$Particles.NormalizedAge
     )
    )
    Module.Index0Float0Write=true
    Module.Index0Float1Write=true
    Module.Index0Float2Write=true
    Module.Index0Float3Write=true
   )
  ]
  rendererClasses [&NiagaraMeshRendererProperties]
  renderers [{rendererClass=&NiagaraMeshRendererProperties, rendererIndex=0}]
 }
 Wipe {
  emitterUpdateScript [
   &EmitterState()
   &SpawnBurst_Instantaneous(Spawn Count=1, Use Spawn Probability=false)
  ]
  particleSpawnScript [
   &CameraOffset(Camera Offset Amount=6) {enabled=false}
   &InitializeParticle(
    Color=<rgba 0.5667 0 2 3>
    Lifetime=&Add_Float(A=$User.LoopTime, B=$User.TimeAdjustment)
    Mass Mode=::NewEnumerator1
    Mesh Scale Mode=::NewEnumerator3
    Position Offset=<-250 0 0>
    Sprite Size Mode=::NewEnumerator3
    Uniform Sprite Size=50
    UsePositionOffset=false
    Write Material Random=true
   ) {module=InitializeParticle001}
  ]
  particleUpdateScript [
   &ParticleState()
   &ScaleMeshSize(Scale Factor=&Multiply_VectorByFloat(
     Vector=<1 1 1>
     Float=&FloatFromCurve(
      FloatCurve=DataInterface() {propertyValues {Curve {
         defaultValue=0
         keys [
          <at 0 1 RCIM_Cubic>
          {arriveTangentWeight=0.06, interpMode=RCIM_Cubic, tangentMode=RCTM_User, time=0.18, value=1}
          {arriveTangent=-1.9231, leaveTangent=-1.9231, time=0.1924, value=0}
          {arriveTangent=1.9231, leaveTangent=1.9231, time=0.8137, value=0}
          {arriveTangentWeight=0.2133, interpMode=RCIM_Cubic, tangentMode=RCTM_User, time=0.82, value=1}
          {arriveTangentWeight=0.0667, interpMode=RCIM_Cubic, tangentMode=RCTM_User, time=1, value=1}
         ]
        }}}
      CurveIndex=&MakeCustomFloatFromBool(
       Bool=$User.End
       True Float=&Add_Float(
        A=$Particles.NormalizedAge
        B=&Subtract_Float(
         A=&MakeCustomFloatFromBool(
          Bool=&SetBoolByFloatComparison(A=0.82, B=$Particles.NormalizedAge, Comparison Type=::NewEnumerator6)
          True Float=$Particles.NormalizedAge
          False Float=0.82
         )
         B=&Divide_Float(A=$System.endtime, B=$System.looptimeInit)
        )
       )
       False Float=$Particles.NormalizedAge
      )
     )
    ))
   &DynamicMaterialParameters(
    Index 0 Param 2=&FloatFromCurve(
     FloatCurve=DataInterface() {propertyValues {Curve {
        defaultValue=0
        keys [
         <at 0 0 RCIM_Cubic>
         {arriveTangentWeight=0.06, interpMode=RCIM_Cubic, tangentMode=RCTM_User, time=0.18, value=1}
         {arriveTangentWeight=0.2133, interpMode=RCIM_Cubic, tangentMode=RCTM_User, time=0.82, value=1}
         {arriveTangentWeight=0.0667, interpMode=RCIM_Cubic, tangentMode=RCTM_User, time=1, value=0}
        ]
       }}}
     CurveIndex=&MakeCustomFloatFromBool(
      Bool=$User.End
      True Float=&Add_Float(
       A=$Particles.NormalizedAge
       B=&Subtract_Float(
        A=&MakeCustomFloatFromBool(
         Bool=&SetBoolByFloatComparison(A=0.82, B=$Particles.NormalizedAge, Comparison Type=::NewEnumerator6)
         True Float=$Particles.NormalizedAge
         False Float=0.82
        )
        B=&Divide_Float(A=$System.endtime, B=$System.looptimeInit)
       )
      )
      False Float=$Particles.NormalizedAge
     )
    )
    Module.Index0Float0Write=true
    Module.Index0Float1Write=true
    Module.Index0Float2Write=true
    Module.Index0Float3Write=true
   )
  ]
  rendererClasses [&NiagaraMeshRendererProperties]
  renderers [{rendererClass=&NiagaraMeshRendererProperties, rendererIndex=0}]
 }
}
summary {
 systemName=NS_Camera_Lens_MistRaven_GrimStrut_Active
 emitters [
  {bEnabled=false, emitterName=InnerFlash, rendererClasses [&NiagaraSpriteRendererProperties]}
  {emitterName=Speedlines, rendererClasses [&NiagaraSpriteRendererProperties]}
  {emitterName=Tendrils, rendererClasses [&NiagaraMeshRendererProperties]}
  {emitterName=Wipe, rendererClasses [&NiagaraMeshRendererProperties]}
 ]
 userVariables [
  User.End {type=NiagaraBool}
  User.LoopTime {type=NiagaraFloat, defaultValue=4, description=""}
  User.TimeAdjustment {type=NiagaraFloat, defaultValue=0.4, description=""}
 ]
}
